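# Template stereo one-shot MetaSound source for multi-product front-end UI sounds, routed to the Fort_UI sound class with UI_MP_FrontEnd concurrency.
# On play, it picks a random wave from the Sounds array, then a random value between Min Pitch and Max Pitch.
# That value is treated as a frequency multiplier and converted to semitones to pitch-shift the wave player, which signals finished when playback ends.
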MetaSoundSource /Game/Sounds/Fort_UI_Sounds/FrontEnd/MultiProduct/Settings_And_Presets/MSS_UI_MP_Template
inputs [
 Max Pitch {type=Float}
 Min Pitch {type=Float}
 Sounds {type=WaveAsset:Array}
 UE.Source.OnPlay {type=Trigger}
]
interfaces [UE.OutputFormat.Stereo:1.0, UE.Source.OneShot:1.0, UE.Source:1.1]
outputs [
 "UE.OutputFormat.Stereo.Audio:0" {type=Audio}
 "UE.OutputFormat.Stereo.Audio:1" {type=Audio}
 UE.Source.OneShot.OnFinished {type=Trigger}
]
pages [Default {
  input_defaults {Max Pitch=1, Min Pitch=1, UE.Source.OnPlay=null, Sounds [null]}
  nodes [
   ArrayRandomGet_0 = Array.Random Get:WaveAsset:Array(In Array=^Sounds, Next=^UE.Source.OnPlay)
   n1 = UE.RandomFloat(Max=^Max Pitch, Min=^Min Pitch, Next=@ArrayRandomGet_0.On Next)
   FrequencyMultipliertoSemitone_0 = UE.Frequency Multiplier to Semitone:Float(Frequency Multiplier=@n1.Value)
   WavePlayer_0 = UE.Wave Player:Stereo(
    Pitch Shift=@FrequencyMultipliertoSemitone_0.Semitones
    Play=@n1.On Next
    Wave Asset=UE.Reroute(Value=UE.Reroute(Value=@ArrayRandomGet_0.Value).Value).Value
   )
  ]
  links [
   {dst="@UE.Source.OneShot.OnFinished", src=WavePlayer_0.On Finished}
   {dst="@UE.OutputFormat.Stereo.Audio:0", src=WavePlayer_0.Out Left}
   {dst="@UE.OutputFormat.Stereo.Audio:1", src=WavePlayer_0.Out Right}
  ]
 }]
properties {
 assetImportData=&self:AssetImportData
 numChannels=2
 outputFormat=Stereo
 soundClassObject=&Fort_UI
 virtualizationMode=Disabled
 concurrencySet [&/UI_MP_FrontEnd]
}
