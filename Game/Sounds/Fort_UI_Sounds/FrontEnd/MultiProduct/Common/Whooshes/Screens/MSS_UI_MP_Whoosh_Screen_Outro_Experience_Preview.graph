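# MetaSound Source preset for the front-end whoosh on the outro of the Experience Preview screen in the MultiProduct UI.
# It inherits its graph from MSS_UI_MP_Template, overriding only inputs: a single whoosh wave in the Sounds array and matching pitch bounds, so no pitch variation.
# Plays as a stereo one-shot in the Fort_UI sound class, with reduced volume, virtualization disabled and the shared UI_MP_FrontEnd concurrency.
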MetaSoundSource /Game/Sounds/Fort_UI_Sounds/FrontEnd/MultiProduct/Common/Whooshes/Screens/MSS_UI_MP_Whoosh_Screen_Outro_Experience_Preview
preset_parent=/Game/Sounds/Fort_UI_Sounds/FrontEnd/MultiProduct/Settings_And_Presets/MSS_UI_MP_Template
inputs [
 Max Pitch {type=Float}
 Min Pitch {type=Float}
 Sounds {type=WaveAsset:Array}
 UE.Source.OnPlay {type=Trigger}
]
interfaces [UE.OutputFormat.Stereo:1.0, UE.Source.OneShot:1.0, UE.Source:1.1]
outputs [
 "UE.OutputFormat.Stereo.Audio:0" {type=Audio}
 "UE.OutputFormat.Stereo.Audio:1" {type=Audio}
 UE.Source.OneShot.OnFinished {type=Trigger}
]
pages [Default {input_defaults {
   Max Pitch=1
   Min Pitch=1
   UE.Source.OnPlay=null
   Sounds [/Game/Sounds/Fort_UI_Sounds/FrontEnd/MultiProduct/Common/Whooshes/Screens/wav/UI_MP_Whoosh_Screen_Outro_PreviewExperience]
  }}]
properties {
 assetImportData=&self:AssetImportData
 numChannels=2
 outputFormat=Stereo
 sampleRate=48000
 soundClassObject=&Fort_UI
 virtualizationMode=Disabled
 volume=0.5
 concurrencySet [/Game/Sounds/Fort_UI_Sounds/FrontEnd/MultiProduct/Settings_And_Presets/UI_MP_FrontEnd]
}
